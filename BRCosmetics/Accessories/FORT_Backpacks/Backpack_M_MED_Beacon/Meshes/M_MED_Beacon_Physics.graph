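# Physics asset for the M_MED Beacon backpack cosmetic, providing collision and joints for its dangling ring and rope bones.
# The pack body combines a capsule and a box, while the ring and two rope bodies each use a capsule.
# The ring joins the pack with twist and one swing axis locked, leaving a single swing axis.
# The first rope chains from the ring, and the second joins the pack directly with smaller limit angles.
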PhysicsAsset /BRCosmetics/Accessories/FORT_Backpacks/Backpack_M_MED_Beacon/Meshes/M_MED_Beacon_Physics
bodies {
 dyn_ring [{shapeType=Capsule, length=5.7551, radius=2, center=<2 0 0>, rotation=<rot 90 0 0>}]
 dyn_rope_01 [{shapeType=Capsule, length=5, radius=2, center=<8 0 0>, rotation=<rot 90 0 0>}]
 dyn_rope_02 [{shapeType=Capsule, length=1.7936, radius=2.1955, center=<7.5166 0.6243 0.2307>, rotation=<rot 90 0 -15.0002>}]
 pack [
  {shapeType=Capsule, length=18.8016, radius=3.7594, center=<1.5799 -4.812 1.1427>, rotation=<rot -3.557 -89.999 19.4072>}
  {shapeType=Box, extentX=23.9359, extentY=7.5, extentZ=12.7349, center=<0 -6.0839 -13.7569>}
 ]
}
constraints [
 {
  bone1Name=dyn_ring
  bone2Name=pack
  swing1LimitDegrees=10
  swing1Motion=Locked
  swing2LimitDegrees=45
  twistLimitDegrees=5
  twistMotion=Locked
 }
 {bone1Name=dyn_rope_01, bone2Name=dyn_ring, swing1LimitDegrees=45, swing2LimitDegrees=45, twistLimitDegrees=10}
 {bone1Name=dyn_rope_02, bone2Name=pack, swing1LimitDegrees=10, swing2LimitDegrees=10, twistLimitDegrees=5}
]
