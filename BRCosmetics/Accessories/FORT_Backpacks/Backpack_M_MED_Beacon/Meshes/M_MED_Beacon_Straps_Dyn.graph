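# Physics asset for the dynamic straps on the M_MED_Beacon backpack cosmetic, letting them swing as physics-driven chains.
# Five capsules approximate the pack body, while small capsules cover the strap bones.
# The master strap bone joins the pack with free swing and twist and carries a tightly constrained base bone.
# Upper and lower four-bone chains branch from that base, with narrow angular limits at each joint to restrain bending and twisting.
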PhysicsAsset /BRCosmetics/Accessories/FORT_Backpacks/Backpack_M_MED_Beacon/Meshes/M_MED_Beacon_Straps_Dyn
bodies {
 dyn_straps_MSTR [{shapeType=Capsule, length=1, radius=1}]
 dyn_straps_base [{shapeType=Capsule, length=1, radius=1}]
 dyn_straps_lw_01 [{shapeType=Capsule, length=3, radius=1.5, center=<2.2032 -1.1906 -1.0571>, rotation=<rot 56.2681 0 22.201>}]
 dyn_straps_lw_02 [{shapeType=Capsule, length=3, radius=1.5, center=<2 0 0>, rotation=<rot 90 0 0>}]
 dyn_straps_lw_03 [{shapeType=Capsule, length=3, radius=1.5, center=<2 0 0>, rotation=<rot 90 0 0>}]
 dyn_straps_lw_04 [{shapeType=Capsule, length=3, radius=1.5, center=<2 0 0>, rotation=<rot 90 0 0>}]
 dyn_straps_up_01 [{shapeType=Capsule, length=3, radius=1.5, center=<1 0 0>, rotation=<rot 90 0 0>}]
 dyn_straps_up_02 [{shapeType=Capsule, length=3, radius=1.5, center=<1 0 0.4945>, rotation=<rot 90 0 0>}]
 dyn_straps_up_03 [{shapeType=Capsule, length=3, radius=1.5, center=<1 0 0>, rotation=<rot 90 0 0>}]
 dyn_straps_up_04 [{shapeType=Capsule, length=3, radius=1.5, center=<1 0 0>, rotation=<rot 90 0 0>}]
 pack [
  {shapeType=Capsule, length=8.4885, radius=3.5125, center=<-4.9383 -5.5353 -15.763>, rotation=<rot -3.557 -89.9985 -72.3012>}
  {shapeType=Capsule, length=9.2225, radius=2.7785, center=<5.1293 -5.7365 -15.7605>, rotation=<rot -3.5569 -89.9971 -102.3005>}
  {shapeType=Capsule, length=23.896, radius=3.4645, center=<2.0291 -4.5073 -0.0685>, rotation=<rot -1.0854 -89.2066 -162.3323>}
  {shapeType=Capsule, length=9.0638, radius=2.9372, center=<-12.2673 -5.6325 -10.7056>, rotation=<rot 0.4235 -93.0258 -37.2168>}
  {shapeType=Capsule, length=9.3006, radius=2.7604, center=<-0.5852 -6.3662 -8.393>, rotation=<rot -1.0854 -89.2066 -162.3323>}
 ]
}
constraints [
 {bone1Name=dyn_straps_lw_01, bone2Name=dyn_straps_base, swing1LimitDegrees=12, swing2LimitDegrees=10, twistLimitDegrees=5}
 {bone1Name=dyn_straps_lw_02, bone2Name=dyn_straps_lw_01, swing1LimitDegrees=12, swing2LimitDegrees=10, twistLimitDegrees=5}
 {bone1Name=dyn_straps_lw_03, bone2Name=dyn_straps_lw_02, swing1LimitDegrees=12, swing2LimitDegrees=10, twistLimitDegrees=5}
 {bone1Name=dyn_straps_lw_04, bone2Name=dyn_straps_lw_03, swing1LimitDegrees=12, swing2LimitDegrees=10, twistLimitDegrees=5}
 {
  bone1Name=dyn_straps_MSTR
  bone2Name=pack
  swing1LimitDegrees=45
  swing1Motion=Free
  swing2LimitDegrees=45
  swing2Motion=Free
  twistLimitDegrees=45
  twistMotion=Free
 }
 {bone1Name=dyn_straps_base, bone2Name=dyn_straps_MSTR, swing1LimitDegrees=5, swing2LimitDegrees=5, twistLimitDegrees=5}
 {bone1Name=dyn_straps_up_01, bone2Name=dyn_straps_base, swing1LimitDegrees=10, swing2LimitDegrees=10, twistLimitDegrees=5}
 {bone1Name=dyn_straps_up_02, bone2Name=dyn_straps_up_01, swing1LimitDegrees=10, swing2LimitDegrees=10, twistLimitDegrees=5}
 {bone1Name=dyn_straps_up_03, bone2Name=dyn_straps_up_02, swing1LimitDegrees=10, swing2LimitDegrees=10, twistLimitDegrees=5}
 {bone1Name=dyn_straps_up_04, bone2Name=dyn_straps_up_03, swing1LimitDegrees=10, swing2LimitDegrees=10, twistLimitDegrees=5}
]
