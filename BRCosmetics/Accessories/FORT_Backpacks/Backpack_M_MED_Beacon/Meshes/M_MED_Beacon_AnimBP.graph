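# Animation Blueprint for the M_MED_Beacon backpack cosmetic, adding procedural flutter and physics to its straps.
# The event graph builds time-based sine noise through Noise FNC, switching most rotation speeds when jumping or falling.
# Noise drives additive offsets on the first upper strap and master bones, then two pack-relative RigidBody nodes simulate body and low-gravity straps.
# Other strap bone modifiers exist but are disconnected from the output pose.
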AnimBlueprint /BRCosmetics/Accessories/FORT_Backpacks/Backpack_M_MED_Beacon/Meshes/M_MED_Beacon_AnimBP
target_skeleton=&/M_MED_Beacon_Skeleton
graphs [
 AnimGraph {nodes [
   "Transform (Modify) Bone - Bone: dyn_straps_lw_01"(Alpha=0.5, Translation="0,0,0", Rotation=Get RotateNoiseC() -> (RotateNoiseC:Rotator), Scale="1,1,1") {rotationMode=BMM_Additive, rotationSpace=BCS_BoneSpace, alphaBoolBlend {blendOption=Linear}, boneToModify {boneName=dyn_straps_lw_01}}
   n1 = Get RotateNoiseD() -> (RotateNoiseD:Rotator)
   n2 = "Transform (Modify) Bone - Bone: dyn_straps_lw_02"(Alpha=-0.3, Translation="0,0,0", Rotation=@n1, Scale="1,1,1") {rotationMode=BMM_Additive, rotationSpace=BCS_BoneSpace, alphaBoolBlend {blendOption=Linear}, boneToModify {boneName=dyn_straps_lw_02}}
   n3 = "Transform (Modify) Bone - Bone: dyn_straps_lw_03"(ComponentPose=@n2, Alpha=-1, Translation="0,0,0", Rotation="0,0,0", Scale="1,1,1") {rotationMode=BMM_Additive, rotationSpace=BCS_BoneSpace, alphaBoolBlend {blendOption=Linear}, boneToModify {boneName=dyn_straps_lw_03}}
   "Transform (Modify) Bone - Bone: dyn_straps_lw_04"(ComponentPose=@n3, Alpha=0.5, Translation="0,0,0", Rotation=@n1, Scale="1,1,1") {rotationMode=BMM_Additive, rotationSpace=BCS_BoneSpace, alphaBoolBlend {blendOption=Linear}, boneToModify {boneName=dyn_straps_lw_04}}
   n4 = "Transform (Modify) Bone - Bone: dyn_straps_up_02"(Alpha=1, Translation="0,0,0", Rotation=Get RotateNoiseB() -> (RotateNoiseB:Rotator), Scale="1,1,1") {rotationMode=BMM_Additive, rotationSpace=BCS_BoneSpace, alphaBoolBlend {blendOption=Linear}, boneToModify {boneName=dyn_straps_up_02}}
   n5 = "Transform (Modify) Bone - Bone: dyn_straps_up_03"(ComponentPose=@n4, Alpha=1, Translation="0,0,0", Rotation=Get RotateNoiseC(), Scale="1,1,1") {rotationMode=BMM_Additive, rotationSpace=BCS_BoneSpace, alphaBoolBlend {blendOption=Linear}, boneToModify {boneName=dyn_straps_up_03}}
   "Transform (Modify) Bone - Bone: dyn_straps_up_04"(ComponentPose=@n5, Alpha=1, Translation="0,0,0", Rotation=Get RotateNoiseD(), Scale="1,1,1") {rotationMode=BMM_Additive, rotationSpace=BCS_BoneSpace, alphaBoolBlend {blendOption=Linear}, boneToModify {boneName=dyn_straps_up_04}}
   n6 = "Transform (Modify) Bone - Bone: dyn_straps_up_01"(
    ComponentPose=Local To Component(LocalPose=Input Pose())
    Alpha=1
    Translation="0,0,0"
    Rotation=Get RotateNoiseA() -> (RotateNoiseA:Rotator)
    Scale="1,1,1"
   ) {rotationMode=BMM_Additive, rotationSpace=BCS_BoneSpace, alphaBoolBlend {blendOption=Linear}, boneToModify {boneName=dyn_straps_up_01}}
   n7 = "Transform (Modify) Bone - Bone: dyn_straps_MSTR"(ComponentPose=@n6, Alpha=1, Translation=Get TranslateNoiseA() -> (TranslateNoiseA:Vector), Rotation="0,0,0", Scale="1,1,1") {translationMode=BMM_Additive, translationSpace=BCS_BoneSpace, alphaBoolBlend {blendOption=Linear}, boneToModify {boneName=dyn_straps_MSTR}}
   n8 = RigidBody(ComponentPose=@n7) {
    bClampLinearTranslationLimitToRefPose=true
    bOverrideWorldGravity=true
    overridePhysicsAsset=&/M_MED_Beacon_Physics
    simulationSpace=BaseBoneSpace
    alphaBoolBlend {blendOption=Linear}
    baseBoneRef {boneName=pack}
    componentAppliedLinearAccClamp=<3000 3000 3000>
    componentLinearAccScale=<0.2 0.2 0.2>
    componentLinearVelScale=<0.2 0.2 0.2>
    overrideWorldGravity {z=-980}
   }
   n9 = RigidBody(ComponentPose=@n8) {
    bClampLinearTranslationLimitToRefPose=true
    bOverrideWorldGravity=true
    overridePhysicsAsset=&/M_MED_Beacon_Straps_Dyn
    simulationSpace=BaseBoneSpace
    alphaBoolBlend {blendOption=Linear}
    baseBoneRef {boneName=pack}
    componentAppliedLinearAccClamp=<500 500 500>
    componentLinearAccScale=<0.01 0.05 0.05>
    componentLinearVelScale {y=0.1, z=0.1}
    overrideWorldGravity {z=-5}
   }
   Output Pose(Result=Component To Local(ComponentPose=@n9)) {layerGroup=None}
  ]}
 EventGraph {nodes [
   Event BlueprintUpdateAnimation()
   > n1 = Sequence!(execute) -> (then_0, then_1, then_2:Exec, then_3:Exec, then_4:Exec, then_5:Exec)
   > n2 = Branch(Condition=OR Boolean(A=Get bIsJumping(), B=Get bIsFalling()))
   > Set OverrideSpeed(OverrideSpeed:Vector="1,12,20") -> (Output_Get:Vector)
   Set OverrideSpeed(execute=@n2.else, OverrideSpeed="0.5,2,1")
   n3 = Get OverrideSpeed() -> (OverrideSpeed:Vector)
   n4 = Noise FNC!(execute=@n1.then_1, self, Value:Vector="2,2,2", Speed:Vector=@n3, Offset:Vector="0, 0, 0") -> (then, Rotator:Rotator)
   > Set RotateNoiseA(RotateNoiseA:Rotator=@n4) -> (Output_Get:Rotator)
   n5 = Noise FNC(execute=@n1.then_2, Value="2,2,5", Speed=@n3, Offset="0,0.5,0.5")
   > Set RotateNoiseB(RotateNoiseB:Rotator=@n5) -> (Output_Get:Rotator)
   n6 = Noise FNC(execute=@n1.then_3, Value="2,5,15", Speed="0.5,1,1.5", Offset="0,-1,-1")
   > Set RotateNoiseC(RotateNoiseC:Rotator=@n6) -> (Output_Get:Rotator)
   n7 = Noise FNC(execute=@n1.then_4, Value="5,15,20", Speed=@n3, Offset="0,2,2")
   > Set RotateNoiseD(RotateNoiseD:Rotator=@n7) -> (Output_Get:Rotator)
   > n8 = Noise FNC(Value="1,0,0", Speed="1.2,0.5,0", Offset="0, 0, 0")
   n9 = BreakRotator(InRot=@n8)
   > Set TranslateNoiseA(TranslateNoiseA:Vector=MakeVector(X=float + float(A=@n9, B=1) {is=title K2Node_CommutativeAssociativeBinaryOperator}, Y=0)) -> (Output_Get:Vector)
   TryGetPawnOwner()
  ]}
 Noise FNC {nodes [
   n1 = Noise FNC!() -> (then, Value:Vector=~, Speed:Vector=~, Offset:Vector=~) {is=title K2Node_FunctionEntry}
   n2 = BreakVector(InVec=vector + float!(A:Vector=@n1.Offset, B:double=GetGameTimeInSeconds()) -> (ReturnValue:Vector))
   n3 = BreakVector(InVec=@n1.Speed)
   n4 = "Sin (Radians)"(A=float * float(A=@n2.Z, B=@n3.Z))
   n5 = BreakVector(InVec=@n1)
   n6 = "Sin (Radians)"(A=float * float(A=@n2, B=@n3))
   n7 = "Sin (Radians)"(A=float * float(A=@n2.Y, B=@n3.Y))
   > Return Node!(
    execute
    Rotator:Rotator=MakeRotator(Roll=float * float(A=@n6, B=@n5), Pitch=float * float(A=@n7, B=@n5.Y), Yaw=float * float(A=@n4, B=@n5.Z))
   )
  ]}
]
variables [
 OverrideSpeed {type=Vector}
 RotateNoiseA {type=Rotator}
 RotateNoiseB {type=Rotator}
 RotateNoiseC {type=Rotator}
 RotateNoiseD {type=Rotator}
 TranslateNoiseA {type=Vector}
]
